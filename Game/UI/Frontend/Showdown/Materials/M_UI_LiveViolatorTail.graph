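# UI material for the arrow tail of a "live violator" element in the Showdown frontend, rendered translucent.
# Color blends from TailColor to HoverColor through the AnimateWhiteFill scalar.
# The tail texture's UVs scale around a point set by the Scale and Offset parameters.
# Its red channel passes through the AAThreshold function as a gradient, giving an antialiased shape tuned by Edge and EdgeSoftness.
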Material /Game/UI/Frontend/Showdown/Materials/M_UI_LiveViolatorTail
graph {
 properties {
  bUsedWithStaticMesh=false
  blendMode=BLEND_Translucent
  materialDomain=MD_UI
  pixelDepthOffsetMode=PDOM_Legacy
  usedShadingModels=Default Lit
 }
 nodes [
  LinearInterpolate(
   A=VectorParameter() {parameterName=TailColor, defaultValue=<rgba 1 1 1 1>}.RGB
   B=VectorParameter() {parameterName=HoverColor, defaultValue=<rgba 1 1 1 1>}.RGB
   Alpha=ScalarParameter() {parameterName=AnimateWhiteFill, sliderMax=1}
  )
  n1 = TextureSampleParameter2D(UVs=MaterialFunctionCall!(
    Texture Scale=ScalarParameter() {defaultValue=1, parameterName=Scale}
    UVs
    Offset XY Texturesize Z=VectorParameter() {parameterName=Offset, defaultValue=<rgba 32 32 64 1>}.RGB
   ) {materialFunction=/Engine/Functions/Engine_MaterialFunctions01/Texturing/ScaleUVsAroundPoint}.UVs) {parameterName=Texture, samplerSource=SSM_Clamp_WorldGroupSettings, texture=&/T_UI_ViolatorArrowTail}
  MaterialFunctionCall!(
   Threshold=ScalarParameter() {defaultValue=0.2, parameterName=Edge}
   Gradient=@n1.R
   Filter Width Override=ScalarParameter() {defaultValue=0.01, parameterName=EdgeSoftness}
   Filter Width Fudge Factor
  ) {materialFunction=/Game/UI/Materials/AAThreshold}
 ]
}
