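# Lobby violator widget for a Showdown tournament, a callout with a tail showing title, playlist tile, status and time remaining, built on FortShowdownDetailView.
# Refresh Data sets the uppercased title and tints text, borders, fill and tail from the tournament's display info.
# Event time state picks the status label and switches between a LIVE indicator and starts/ended text.
# Until runOnce is set, refresh waits a second, plays the scale-in Intro and reveals the collapsed container.
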WidgetBlueprint /Game/UI/Frontend/Showdown/ShowdownLobbyViolator
animations [Intro {
  binding_count=5
  display_label=""
  length_seconds=5.1091
  object=&self:Intro
  sequence {bindings [
    Overlay_0 {
     id=#1
     possessed_class=Overlay
     tracks [{
       type=MovieScene2DTransformTrack
       sections [{
         type=MovieScene2DTransformSection
         channels [
          Scale.X {type=MovieSceneScriptingFloatChannel, keys [[0,0],[6.0357,1.0841],[6.8237,1],[7.4053,1]]}
          Scale.Y {type=MovieSceneScriptingFloatChannel, keys [[0,0],[6.0357,1.0841],[6.8237,1],[7.4053,1]]}
         ]
        }]
      }]
    }
    SizeBoxGameModeImage {
     id=#2
     possessed_class=SizeBox
     tracks [
      {
       type=MovieSceneFloatTrack
       display_name=Height Override
       sections [{type=MovieSceneFloatSection, channels [None {type=MovieSceneScriptingFloatChannel, const [100,1]}]}]
      }
      {
       type=MovieSceneFloatTrack
       display_name=Render Opacity
       sections [{type=MovieSceneFloatSection, channels [None {type=MovieSceneScriptingFloatChannel, const [0.2,1]}]}]
      }
     ]
    }
    BackgroundBlur_0 {
     id=#3
     possessed_class=BackgroundBlur
     tracks [{
       type=MovieSceneFloatTrack
       display_name=Blur Strength
       sections [{type=MovieSceneFloatSection, end_frame=102, start_frame=0, channels [None {type=MovieSceneScriptingFloatChannel, const [20,1]}]}]
      }]
    }
    Text_ShowdownTitle {
     id=#4
     possessed_class=CommonTextBlock
     tracks [{
       type=MovieScene2DTransformTrack
       sections [{
         type=MovieScene2DTransformSection
         channels [
          Translation.X {type=MovieSceneScriptingFloatChannel}
          Translation.Y {type=MovieSceneScriptingFloatChannel, const [0,1]}
         ]
        }]
      }]
    }
    Image_BGFillColor {
     id=#5
     possessed_class=Image
     tracks [{
       type=MovieSceneFloatTrack
       display_name=Render Opacity
       sections [{type=MovieSceneFloatSection, channels [None {type=MovieSceneScriptingFloatChannel, const [0.5,1]}]}]
      }]
    }
   ]}
 }]
graphs [EventGraph {nodes [
   Event Refresh Data() {is=title K2Node_Event}
   > n1 = Sequence()
   n2 = Break Fort Tournament Display Info!(FortTournamentDisplayInfo:"Fort Tournament Display Info Structure (by ref)"=Get TournamentDisplayInfo() -> (TournamentDisplayInfo:Fort Tournament Display Info Structure)) -> (
    TitleLine1:Text
    TitleLine2:Text
    ScheduleInfo:Text
    FlavorDescription:Text
    DetailsDescription:Text
    ShortFormatTitle:Text
    LongFormatTitle:Text
    BackgroundTitle:Text
    PinScoreRequirement:int
    PinEarnedText:Text
    BaseColor:Linear Color Structure
    PrimaryColor:Linear Color Structure
    SecondaryColor:Linear Color Structure
    HighlightColor:Linear Color Structure
    TitleColor:Linear Color Structure
    ShadowColor:Linear Color Structure
    BackgroundLeftColor:Linear Color Structure
    BackgroundRightColor:Linear Color Structure
    BackgroundTextColor:Linear Color Structure
    PosterFadeColor:Linear Color Structure
    AlertText:Text
    SeriesPointLeaderboardName:Text
    AlertType:EFortTournamentAlertType Enum
    RoundDisplayNames:Array of Texts
   ) {is=title K2Node_BreakStruct}
   n3 = Get Text_ShowdownTitle() -> (Text_ShowdownTitle:Common Text Object Reference)
   > "SetText (Text)"(
    self=@n3
    InText=TextToUpper!(InText:"Text (by ref)"=Format Text!(Format="LOCTEXT(\"{TitleLine1} {TitleLine2}\")", TitleLine1:Text=@n2, TitleLine2:Text=@n2.TitleLine2) -> (Result)) -> (ReturnValue:Text)
   )
   > SetColorAndOpacity(self=@n3, InColorAndOpacity=Make Slate Color(SpecifiedColor=@n2.PrimaryColor))
   > SetBrushColor(self=Get CommonBorder_DetailsBG() -> (CommonBorder_DetailsBG:Common Border Object Reference), InBrushColor=@n2.ShadowColor)
   > SetColorAndOpacity(
    self:Image Object Reference=[
     Get Image_OuterTriangle() -> (Image_OuterTriangle:Image Object Reference)
     Get Image_InnerTriangle() -> (Image_InnerTriangle:Image Object Reference)
    ]
    InColorAndOpacity:Linear Color Structure=@n2.ShadowColor
   )
   n4 = Get CommonBorder_HighlightBorder() -> (CommonBorder_HighlightBorder:Common Border Object Reference)
   > SetBrushColor(self=@n4, InBrushColor=@n2.HighlightColor)
   > SetContentColorAndOpacity!(execute, self:Border Object Reference=@n4, InContentColorAndOpacity:Linear Color Structure=@n2.HighlightColor) -> (then)
   > SetColorAndOpacity(self=Get Image_BGFillColor() -> (Image_BGFillColor:Image Object Reference), InColorAndOpacity=@n2.BackgroundRightColor)
   n5 = GetEventTimeState!(self:Fort Showdown Detail View Object Reference) -> (ReturnValue:EFortShowdownEventState Enum=Unknown)
   > "SetText (Text)"(
    self=Get CommonTextBlock_TimeRemainingDesc() -> (CommonTextBlock_TimeRemainingDesc:Common Text Object Reference)
    InText=TextToUpper(InText=Select!(
      Unknown:Text
      FutureTBD:Text
      FutureScheduled:Text="LOCTEXT(\"Starts In\")"
      FutureNext:Text="LOCTEXT(\"Starts In \")"
      Live:Text="LOCTEXT(\"Ends In\")"
      LiveParticipating:Text="LOCTEXT(\"Ends In\")"
      LiveNotParticipating:Text="LOCTEXT(\"Ends In\")"
      Completed:Text="LOCTEXT(\"ENDED\")"
      CompletedParticipated:Text="LOCTEXT(\"ENDED\")"
      CompletedNotPartipated:Text="LOCTEXT(\"ENDED\")"
      Cancelled:Text
      Index:EFortShowdownEventState Enum=@n5
     ) -> (ReturnValue:Text))
   )
   > SetActiveWidgetIndex(
    self=Get CommonWidgetSwitcher_Status() -> (CommonWidgetSwitcher_Status:Common Widget Switcher Legacy Object Reference)
    Index=Select(
     Unknown:int=0
     FutureTBD:int=1
     FutureScheduled:int=1
     FutureNext:int=1
     Live:int=0
     LiveParticipating:int=0
     LiveNotParticipating:int=0
     Completed:int=1
     CompletedParticipated:int=1
     CompletedNotPartipated:int=1
     Cancelled:int=0
     Index=@n5
    ) -> (ReturnValue:int)
   )
   n6 = Branch(execute=@n1.then_1, Condition=Get runOnce() -> (runOnce:bool))
   Delay(execute=@n6.else, Duration=1)
   > PlayAnimation(InAnimation=Get Intro())
   > SetVisibility(self=Get CommonBorder_TopContainer() -> (CommonBorder_TopContainer:Common Border Object Reference), InVisibility=HitTestInvisible)
   > Set runOnce(runOnce:bool=true) -> (Output_Get:bool)
  ]}]
variables [runOnce {type=boolean}]
widgets {
 info {inheritedWidgetCount=6, parentClass=/Script/FortniteUI.FortShowdownDetailView, rootWidgetClass=&CommonBorder, widgetCount=37}
 widgets [
  CommonBorder_TopContainer {
   widgetClassPath=&CommonBorder
   bIsVariable=true
   widget=CommonBorder_TopContainer
   properties {verticalAlignment=VAlign_Bottom, visibility=Collapsed, padding=<ltrb 0 0 0 25>}
  }
  Overlay_0 {
   widgetClassPath=&Overlay
   parent=CommonBorder_TopContainer
   widget=Overlay_0
   properties {renderTransformPivot=<0.8 1>}
   slot {class=BorderSlot, properties {verticalAlignment=VAlign_Bottom, padding=<ltrb 0 0 0 25>}}
  }
  SizeBox_0 {
   widgetClassPath=&SizeBox
   parent=Overlay_0
   widget=SizeBox_0
   properties {
    bOverride_MaxDesiredHeight=true
    bOverride_WidthOverride=true
    heightOverride=400
    maxDesiredHeight=300
    maxDesiredWidth=384
    minDesiredWidth=192
    widthOverride=384
   }
   slot {class=OverlaySlot, properties {horizontalAlignment=HAlign_Center, verticalAlignment=VAlign_Center}}
  }
  Overlay_1 {widgetClassPath=&Overlay, parent=SizeBox_0, widget=Overlay_1, slot {class=SizeBoxSlot}}
  CommonBorder_HighlightBorder {
   widgetClassPath=&CommonBorder
   bIsVariable=true
   parent=Overlay_1
   widget=CommonBorder_HighlightBorder
   properties {contentColorAndOpacity=<rgba 0 0.4168 1 1>, padding=<ltrb 0 0 0 0>}
   slot {class=OverlaySlot, properties {horizontalAlignment=HAlign_Fill, verticalAlignment=VAlign_Fill}}
  }
  CommonBorder_InnerHighlight {
   widgetClassPath=&CommonBorder
   parent=CommonBorder_HighlightBorder
   widget=CommonBorder_InnerHighlight
   properties {
    background {
     bIsDynamicallyLoaded=false
     drawAs=Box
     imageType=NoImage
     mirroring=NoMirror
     resourceName=None
     resourceObject=/Game/UI/Frontend/Showdown/Materials/M_UI_SkewBoxBorder_Tournament_ModeAlertBorder_Highlight
     tiling=NoTile
     imageSize=<128 32>
     margin=<ltrb 0.1 0.1 0.1 0.1>
     outlineSettings {
      bUseBrushTransparency=false
      roundingType=HalfHeightRadius
      width=0
      color {colorUseRule=UseColor_Specified, specifiedColor=<rgba 0 0 0 0>}
      cornerRadii=<0 0 0 0>
     }
     tintColor {colorUseRule=UseColor_Specified, specifiedColor=<rgba 1 1 1 1>}
     uVRegion {bIsValid=false, max=<0 0>, min=<0 0>}
    }
    padding=<ltrb 3 3 3 3>
    renderTransform {angle=0, scale=<1 1>, shear=<0 0>, translation=<4 4>}
   }
   slot {class=BorderSlot, properties {padding=<ltrb 0 0 0 0>}}
  }
  Overlay_4 {
   widgetClassPath=&Overlay
   parent=CommonBorder_InnerHighlight
   widget=Overlay_4
   slot {class=BorderSlot, properties {padding=<ltrb 3 3 3 3>}}
  }
  Spacer_0 {
   widgetClassPath=&Spacer
   parent=Overlay_4
   widget=Spacer_0
   properties {size=<1 99.2413>}
   slot {class=OverlaySlot, properties {verticalAlignment=VAlign_Fill}}
  }
  Overlay_5 {
   widgetClassPath=&Overlay
   parent=Overlay_4
   widget=Overlay_5
   properties {renderTransform {angle=0, scale=<1 1>, shear=<0 0>, translation=<120 0>}}
   slot {class=OverlaySlot, properties {horizontalAlignment=HAlign_Center, verticalAlignment=VAlign_Bottom}}
  }
  Image_HighlightTail {
   widgetClassPath=&Image
   parent=Overlay_5
   widget=Image_HighlightTail
   properties {
    visibility=SelfHitTestInvisible
    brush {
     bIsDynamicallyLoaded=false
     drawAs=Image
     imageType=NoImage
     mirroring=NoMirror
     resourceName=None
     resourceObject=/Game/UI/Frontend/Showdown/Materials/M_UI_SkewBoxBorder_Tournament_ModeAlertBorder_Tail
     tiling=NoTile
     imageSize=<40 40>
     margin=<ltrb 0 0 0 0>
     outlineSettings {
      bUseBrushTransparency=false
      roundingType=HalfHeightRadius
      width=0
      color {colorUseRule=UseColor_Specified, specifiedColor=<rgba 0 0 0 0>}
      cornerRadii=<0 0 0 0>
     }
     tintColor {colorUseRule=UseColor_Specified, specifiedColor=<rgba 1 1 1 1>}
     uVRegion {bIsValid=false, max=<0 0>, min=<0 0>}
    }
    renderTransform {angle=0, scale=<1 1>, shear=<0 0>, translation=<0 26>}
   }
   slot {class=OverlaySlot}
  }
  CommonBorder_InnerContent {
   widgetClassPath=&CommonBorder
   parent=Overlay_1
   widget=CommonBorder_InnerContent
   properties {
    background {
     bIsDynamicallyLoaded=false
     drawAs=NoDrawType
     imageType=NoImage
     mirroring=NoMirror
     resourceName=None
     resourceObject=/Game/UI/Frontend/Showdown/Materials/M_UI_SkewBoxBorder_Tournament_ModeAlertBorder
     tiling=NoTile
     imageSize=<128 32>
     margin=<ltrb 0.1 0.1 0.1 0.1>
     outlineSettings {
      bUseBrushTransparency=false
      roundingType=HalfHeightRadius
      width=0
      color {colorUseRule=UseColor_Specified, specifiedColor=<rgba 0 0 0 0>}
      cornerRadii=<0 0 0 0>
     }
     tintColor {colorUseRule=UseColor_Specified, specifiedColor=<rgba 1 1 1 1>}
     uVRegion {bIsValid=false, max=<0 0>, min=<0 0>}
    }
    padding=<ltrb 0 0 0 0>
   }
   slot {class=OverlaySlot, properties {horizontalAlignment=HAlign_Fill, verticalAlignment=VAlign_Fill}}
  }
  Overlay_3 {
   widgetClassPath=&Overlay
   parent=CommonBorder_InnerContent
   widget=Overlay_3
   slot {class=BorderSlot, properties {padding=<ltrb 0 0 0 0>}}
  }
  SizeBoxGameModeImage {
   widgetClassPath=&SizeBox
   parent=Overlay_3
   widget=SizeBoxGameModeImage
   properties {bOverride_HeightOverride=true, clipping=ClipToBounds, heightOverride=100, widthOverride=400}
   slot {class=OverlaySlot, properties {horizontalAlignment=HAlign_Fill, verticalAlignment=VAlign_Bottom}}
  }
  ScaleBox_1 {
   widgetClassPath=&ScaleBox
   parent=SizeBoxGameModeImage
   widget=ScaleBox_1
   properties {stretch=ScaleToFitX}
   slot {class=SizeBoxSlot}
  }
  Image_PlaylistTile {
   widgetClassPath=/Script/EpicCMSUIFramework.EpicCMSImage
   bInherited=true
   bIsVariable=true
   parent=ScaleBox_1
   widget=Image_PlaylistTile
   properties {brush {
     bIsDynamicallyLoaded=false
     drawAs=Image
     imageType=NoImage
     mirroring=NoMirror
     resourceName=None
     resourceObject=null
     tiling=NoTile
     imageSize=<400 200>
     margin=<ltrb 0 0 0 0>
     outlineSettings {
      bUseBrushTransparency=false
      roundingType=HalfHeightRadius
      width=0
      color {colorUseRule=UseColor_Specified, specifiedColor=<rgba 0 0 0 0>}
      cornerRadii=<0 0 0 0>
     }
     tintColor {colorUseRule=UseColor_Specified, specifiedColor=<rgba 1 1 1 1>}
     uVRegion {bIsValid=false, max=<0 0>, min=<0 0>}
    }}
   slot {class=ScaleBoxSlot}
  }
  BackgroundBlur_0 {
   widgetClassPath=&BackgroundBlur
   parent=Overlay_3
   widget=BackgroundBlur_0
   properties {
    blurStrength=12
    lowQualityFallbackBrush {
     bIsDynamicallyLoaded=false
     drawAs=NoDrawType
     imageType=NoImage
     mirroring=NoMirror
     resourceName=None
     resourceObject=null
     tiling=NoTile
     imageSize=<0 25.7474>
     margin=<ltrb 0 0 0 0>
     outlineSettings {
      bUseBrushTransparency=false
      roundingType=HalfHeightRadius
      width=0
      color {colorUseRule=UseColor_Specified, specifiedColor=<rgba 0 0 0 0>}
      cornerRadii=<0 0 0 0>
     }
     tintColor {colorUseRule=UseColor_Specified, specifiedColor=<rgba 1 1 1 1>}
     uVRegion {bIsValid=false, max=<0 0>, min=<0 0>}
    }
   }
   slot {class=OverlaySlot, properties {horizontalAlignment=HAlign_Fill, verticalAlignment=VAlign_Fill}}
  }
  Image_BGFillColor {
   widgetClassPath=&Image
   bIsVariable=true
   parent=Overlay_3
   widget=Image_BGFillColor
   properties {renderOpacity=0}
   slot {class=OverlaySlot, properties {horizontalAlignment=HAlign_Fill, verticalAlignment=VAlign_Fill}}
  }
  OverlayTail {
   widgetClassPath=&Overlay
   parent=Overlay_3
   widget=OverlayTail
   properties {renderTransform {angle=0, scale=<1 1>, shear=<0 0>, translation=<120 -5>}}
   slot {class=OverlaySlot, properties {horizontalAlignment=HAlign_Center, verticalAlignment=VAlign_Bottom, padding=<ltrb 0 0 0 -4>}}
  }
  Image_OuterTriangle {
   widgetClassPath=&Image
   bIsVariable=true
   parent=OverlayTail
   widget=Image_OuterTriangle
   properties {
    visibility=SelfHitTestInvisible
    brush {
     bIsDynamicallyLoaded=false
     drawAs=Image
     imageType=NoImage
     mirroring=NoMirror
     resourceName=None
     resourceObject=&M_UI_SkewBoxBorder_Tournament_ModeAlertBorder_Tail
     tiling=NoTile
     imageSize=<40 40>
     margin=<ltrb 0 0 0 0>
     outlineSettings {
      bUseBrushTransparency=false
      roundingType=HalfHeightRadius
      width=0
      color {colorUseRule=UseColor_Specified, specifiedColor=<rgba 0 0 0 0>}
      cornerRadii=<0 0 0 0>
     }
     tintColor {colorUseRule=UseColor_Specified, specifiedColor=<rgba 1 1 1 1>}
     uVRegion {bIsValid=false, max=<0 0>, min=<0 0>}
    }
    renderTransform {angle=0, scale=<1 1>, shear=<0 0>, translation=<0 28>}
   }
   slot {class=OverlaySlot}
  }
  Image_InnerTriangle {
   widgetClassPath=&Image
   bIsVariable=true
   parent=OverlayTail
   widget=Image_InnerTriangle
   properties {
    visibility=SelfHitTestInvisible
    brush {
     bIsDynamicallyLoaded=false
     drawAs=Image
     imageType=NoImage
     mirroring=NoMirror
     resourceName=None
     resourceObject=&M_UI_SkewBoxBorder_Tournament_ModeAlertBorder_Tail
     tiling=NoTile
     imageSize=<32 32>
     margin=<ltrb 0 0 0 0>
     outlineSettings {
      bUseBrushTransparency=false
      roundingType=HalfHeightRadius
      width=0
      color {colorUseRule=UseColor_Specified, specifiedColor=<rgba 0 0 0 0>}
      cornerRadii=<0 0 0 0>
     }
     tintColor {colorUseRule=UseColor_Specified, specifiedColor=<rgba 1 1 1 1>}
     uVRegion {bIsValid=false, max=<0 0>, min=<0 0>}
    }
    colorAndOpacity=<rgba 0.0009 0.8632 0.8388 1>
    renderTransform {angle=0, scale=<1 1>, shear=<0 0>, translation=<3 28>}
   }
   slot {class=OverlaySlot}
  }
  VerticalBox_0 {
   widgetClassPath=&VerticalBox
   parent=Overlay_3
   widget=VerticalBox_0
   slot {class=OverlaySlot, properties {horizontalAlignment=HAlign_Fill, verticalAlignment=VAlign_Fill}}
  }
  ScaleBox_0 {
   widgetClassPath=&ScaleBox
   parent=VerticalBox_0
   widget=ScaleBox_0
   properties {clipping=ClipToBounds, stretchDirection=UpOnly}
   slot {class=VerticalBoxSlot, properties {padding=<ltrb 12 0 12 0>, size {sizeRule=Fill, value=1}}}
  }
  Text_ShowdownTitle {
   widgetClassPath=&CommonTextBlock
   bIsVariable=true
   parent=ScaleBox_0
   widget=Text_ShowdownTitle
   properties {
    scrollStyle=&TextScrollStyle-Base_C
    text=TOURNAMENT NAME
    colorAndOpacity {colorUseRule=UseColor_Specified, specifiedColor=<rgba 0 0 0 1>}
    font {
     bForceMonospaced=false
     bMaterialIsStencil=false
     fontMaterial=null
     fontObject=&Burbank
     letterSpacing=0
     monospacedWidth=1
     size=24
     skewAmount=0
     typefaceFontName=Burbank-Black
     outlineSettings {
      bApplyOutlineToDropShadows=false
      bMiteredCorners=false
      bSeparateFillAlpha=false
      outlineMaterial=null
      outlineSize=0
      outlineColor=<rgba 0 0 0 1>
     }
    }
    margin=<ltrb 0 -5 0 -5>
   }
   slot {class=ScaleBoxSlot, properties {horizontalAlignment=HAlign_Fill}}
  }
  CommonBorder_DetailsBG {
   widgetClassPath=&CommonBorder
   bIsVariable=true
   parent=VerticalBox_0
   widget=CommonBorder_DetailsBG
   properties {
    background {
     bIsDynamicallyLoaded=false
     drawAs=Image
     imageType=NoImage
     mirroring=NoMirror
     resourceName=None
     resourceObject=null
     tiling=NoTile
     imageSize=<32 32>
     margin=<ltrb 0 0 0 0>
     outlineSettings {
      bUseBrushTransparency=false
      roundingType=HalfHeightRadius
      width=0
      color {colorUseRule=UseColor_Specified, specifiedColor=<rgba 0 0 0 0>}
      cornerRadii=<0 0 0 0>
     }
     tintColor {colorUseRule=UseColor_Specified, specifiedColor=<rgba 1 1 1 1>}
     uVRegion {bIsValid=false, max=<0 0>, min=<0 0>}
    }
    brushColor=<rgba 0.2035 0 1 1>
    padding=<ltrb 12 0 12 0>
   }
   slot {class=VerticalBoxSlot}
  }
  HorizontalBox_3 {
   widgetClassPath=&HorizontalBox
   parent=CommonBorder_DetailsBG
   widget=HorizontalBox_3
   slot {class=BorderSlot, properties {padding=<ltrb 12 0 12 0>}}
  }
  CommonWidgetSwitcher_Status {
   widgetClassPath=&CommonWidgetSwitcherLegacy
   bIsVariable=true
   parent=HorizontalBox_3
   widget=CommonWidgetSwitcher_Status
   slot {class=HorizontalBoxSlot, properties {horizontalAlignment=HAlign_Left, verticalAlignment=VAlign_Center}}
  }
  HorizontalBox_4 {widgetClassPath=&HorizontalBox, parent=CommonWidgetSwitcher_Status, widget=HorizontalBox_4, slot {class=WidgetSwitcherSlot}}
  Image_RecDot {
   widgetClassPath=&Image
   parent=HorizontalBox_4
   widget=Image_RecDot
   properties {brush {
     bIsDynamicallyLoaded=false
     drawAs=Image
     imageType=NoImage
     mirroring=NoMirror
     resourceName=None
     resourceObject=/Game/UI/Materials/M_UICircle
     tiling=NoTile
     imageSize=<16 16>
     margin=<ltrb 0 0 0 0>
     outlineSettings {
      bUseBrushTransparency=false
      roundingType=HalfHeightRadius
      width=0
      color {colorUseRule=UseColor_Specified, specifiedColor=<rgba 0 0 0 0>}
      cornerRadii=<0 0 0 0>
     }
     tintColor {colorUseRule=UseColor_Specified, specifiedColor=<rgba 1 1 1 1>}
     uVRegion {bIsValid=false, max=<0 0>, min=<0 0>}
    }}
   slot {class=HorizontalBoxSlot, properties {horizontalAlignment=HAlign_Center, verticalAlignment=VAlign_Center, padding=<ltrb 0 0 5 0>}}
  }
  CommonTextBlock_LIVE {
   widgetClassPath=&CommonTextBlock
   parent=HorizontalBox_4
   widget=CommonTextBlock_LIVE
   properties {
    style=/FortUILibrary/Styles/Typography/Foundation/TextStyle-Header-S.TextStyle-Header-S_C
    text=LIVE
    font {
     bForceMonospaced=false
     bMaterialIsStencil=false
     fontMaterial=null
     fontObject=&Burbank
     letterSpacing=0
     monospacedWidth=1
     size=24
     skewAmount=0
     typefaceFontName=Burbank-Black
     outlineSettings {
      bApplyOutlineToDropShadows=false
      bMiteredCorners=false
      bSeparateFillAlpha=false
      outlineMaterial=null
      outlineSize=0
      outlineColor=<rgba 0 0 0 1>
     }
    }
    renderTransform {angle=0, scale=<1 1>, shear=<-8 0>, translation=<0 0>}
    shadowOffset=<0 0>
   }
   slot {class=HorizontalBoxSlot}
  }
  CommonTextBlock_TimeRemainingDesc {
   widgetClassPath=&CommonTextBlock
   bIsVariable=true
   parent=CommonWidgetSwitcher_Status
   widget=CommonTextBlock_TimeRemainingDesc
   properties {
    style=&TextStyle-Header-S_C
    text=STARTS IN
    font {
     bForceMonospaced=false
     bMaterialIsStencil=false
     fontMaterial=null
     fontObject=&Burbank
     letterSpacing=0
     monospacedWidth=1
     size=24
     skewAmount=0
     typefaceFontName=Burbank-Black
     outlineSettings {
      bApplyOutlineToDropShadows=false
      bMiteredCorners=false
      bSeparateFillAlpha=false
      outlineMaterial=null
      outlineSize=0
      outlineColor=<rgba 0 0 0 1>
     }
    }
    renderTransform {angle=0, scale=<1 1>, shear=<-8 0>, translation=<0 0>}
    shadowOffset=<0 0>
   }
   slot {class=WidgetSwitcherSlot, properties {verticalAlignment=VAlign_Center}}
  }
  HorizontalBox_0 {
   widgetClassPath=&HorizontalBox
   parent=HorizontalBox_3
   widget=HorizontalBox_0
   slot {class=HorizontalBoxSlot, properties {horizontalAlignment=HAlign_Right, size {sizeRule=Fill, value=0.5}}}
  }
  Image_1 {
   widgetClassPath=&Image
   parent=HorizontalBox_0
   widget=Image_1
   properties {
    visibility=SelfHitTestInvisible
    brush {
     bIsDynamicallyLoaded=false
     drawAs=Image
     imageType=NoImage
     mirroring=NoMirror
     resourceName=None
     resourceObject=/Game/UI/Foundation/Textures/Icons/Replay/T-Icon-Replay-Timeline-L
     tiling=NoTile
     imageSize=<20 20>
     margin=<ltrb 0 0 0 0>
     outlineSettings {
      bUseBrushTransparency=false
      roundingType=HalfHeightRadius
      width=0
      color {colorUseRule=UseColor_Specified, specifiedColor=<rgba 0 0 0 0>}
      cornerRadii=<0 0 0 0>
     }
     tintColor {colorUseRule=UseColor_Specified, specifiedColor=<rgba 1 1 1 1>}
     uVRegion {bIsValid=false, max=<0 0>, min=<0 0>}
    }
   }
   slot {class=HorizontalBoxSlot, properties {horizontalAlignment=HAlign_Center, verticalAlignment=VAlign_Center, padding=<ltrb 0 0 5 0>}}
  }
  Text_TimeRemaining {
   widgetClassPath=&CommonTextBlock
   bInherited=true
   bIsVariable=true
   parent=HorizontalBox_0
   widget=Text_TimeRemaining
   properties {
    style=&TextStyle-Header-S_C
    text=12:34:15
    font {
     bForceMonospaced=false
     bMaterialIsStencil=false
     fontMaterial=null
     fontObject=&Burbank
     letterSpacing=0
     monospacedWidth=1
     size=24
     skewAmount=0
     typefaceFontName=Burbank-Black
     outlineSettings {
      bApplyOutlineToDropShadows=false
      bMiteredCorners=false
      bSeparateFillAlpha=false
      outlineMaterial=null
      outlineSize=0
      outlineColor=<rgba 0 0 0 1>
     }
    }
    renderTransform {angle=0, scale=<1 1>, shear=<-8 0>, translation=<0 0>}
    shadowOffset=<0 0>
   }
   slot {class=HorizontalBoxSlot, properties {verticalAlignment=VAlign_Center}}
  }
  Text_DebugId {widgetClassPath=&TextBlock, bInherited=true, widget=null, properties=~}
  Image_PosterFront {widgetClassPath=&EpicCMSImage, bInherited=true, widget=null}
  Image_PosterBack {widgetClassPath=&EpicCMSImage, bInherited=true, widget=null}
  Image_LoadingScreen {widgetClassPath=&EpicCMSImage, bInherited=true, widget=null}
 ]
}
