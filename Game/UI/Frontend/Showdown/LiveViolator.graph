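# LiveViolator is a frontend Showdown UI widget that shows a red, skewed "LIVE!" callout badge with a recording dot and a tail.
# Red material-driven skew-box and tail images form the background, duplicated inside a dark, translucent, offset and sheared CommonBorder as a shadow layer.
# A horizontal box on top holds the circle dot and sheared Burbank text.
# PreConstruct sets the text from the LiveText variable, which defaults to "LIVE!".
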WidgetBlueprint /Game/UI/Frontend/Showdown/LiveViolator
graphs [EventGraph {nodes [
   Event PreConstruct()
   > "SetText (Text)"(self=Get Text_Callout() -> (Text_Callout:Common Text Object Reference), InText=Get LiveText() -> (LiveText:Text))
  ]}]
variables [LiveText {type=string, default="LIVE!"}]
widgets {
 info {inheritedWidgetCount=0, parentClass=&UserWidget, rootWidgetClass=&Overlay, widgetCount=11}
 widgets [
  Overlay_1 {widgetClassPath=&Overlay, widget=Overlay_1}
  CommonBorderShadow {
   widgetClassPath=&CommonBorder
   parent=Overlay_1
   widget=CommonBorderShadow
   properties {
    contentColorAndOpacity=<rgba 0.0052 0.0052 0.0056 0.4>
    padding=<ltrb 0 0 0 0>
    renderTransform {angle=2, scale=<1 1>, shear=<-2 0>, translation=<0 5>}
    renderTransformPivot=<0.1813 1.931>
   }
   slot {class=OverlaySlot, properties {horizontalAlignment=HAlign_Fill, verticalAlignment=VAlign_Bottom, padding=<ltrb 0 -5 0 0>}}
  }
  OverlayShadow {
   widgetClassPath=&Overlay
   parent=CommonBorderShadow
   widget=OverlayShadow
   slot {class=BorderSlot, properties {padding=<ltrb 0 0 0 0>}}
  }
  Image_0 {
   widgetClassPath=&Image
   parent=OverlayShadow
   widget=Image_0
   properties {
    brush {
     bIsDynamicallyLoaded=false
     drawAs=Image
     imageType=NoImage
     mirroring=NoMirror
     resourceName=None
     resourceObject=/Game/UI/Frontend/Showdown/Materials/M_UI_LiveViolatorTail_Inst
     tiling=NoTile
     imageSize=<64 64>
     margin=<ltrb 0 0 0 0>
     outlineSettings {
      bUseBrushTransparency=false
      roundingType=HalfHeightRadius
      width=0
      color {colorUseRule=UseColor_Specified, specifiedColor=<rgba 0 0 0 0>}
      cornerRadii=<0 0 0 0>
     }
     tintColor {colorUseRule=UseColor_Specified, specifiedColor=<rgba 1 1 1 1>}
     uVRegion {bIsValid=false, max=<0 0>, min=<0 0>}
    }
    colorAndOpacity=<rgba 1 0 0 1>
    renderTransform {angle=0, scale=<1 1>, shear=<0 0>, translation=<-39 51>}
   }
   slot {class=OverlaySlot, properties {horizontalAlignment=HAlign_Center, verticalAlignment=VAlign_Center}}
  }
  Image_1 {
   widgetClassPath=&Image
   parent=OverlayShadow
   widget=Image_1
   properties {
    brush {
     bIsDynamicallyLoaded=false
     drawAs=Image
     imageType=NoImage
     mirroring=NoMirror
     resourceName=None
     resourceObject=/Game/UI/Frontend/Showdown/Materials/M_UI_SkewBox_LiveBG
     tiling=NoTile
     imageSize=<32 32>
     margin=<ltrb 0 0 0 0>
     outlineSettings {
      bUseBrushTransparency=false
      roundingType=HalfHeightRadius
      width=0
      color {colorUseRule=UseColor_Specified, specifiedColor=<rgba 0 0 0 0>}
      cornerRadii=<0 0 0 0>
     }
     tintColor {colorUseRule=UseColor_Specified, specifiedColor=<rgba 1 1 1 1>}
     uVRegion {bIsValid=false, max=<0 0>, min=<0 0>}
    }
    colorAndOpacity=<rgba 1 0 0 1>
    renderTransform {angle=0, scale=<1 1>, shear=<-14 0>, translation=<0 0>}
   }
   slot {class=OverlaySlot, properties {horizontalAlignment=HAlign_Fill, verticalAlignment=VAlign_Fill}}
  }
  Overlay_3 {
   widgetClassPath=&Overlay
   parent=Overlay_1
   widget=Overlay_3
   slot {class=OverlaySlot, properties {horizontalAlignment=HAlign_Fill, verticalAlignment=VAlign_Bottom, padding=<ltrb 0 -5 0 0>}}
  }
  Image_Tail {
   widgetClassPath=&Image
   parent=Overlay_3
   widget=Image_Tail
   properties {
    brush {
     bIsDynamicallyLoaded=false
     drawAs=Image
     imageType=NoImage
     mirroring=NoMirror
     resourceName=None
     resourceObject=&M_UI_LiveViolatorTail_Inst
     tiling=NoTile
     imageSize=<64 64>
     margin=<ltrb 0 0 0 0>
     outlineSettings {
      bUseBrushTransparency=false
      roundingType=HalfHeightRadius
      width=0
      color {colorUseRule=UseColor_Specified, specifiedColor=<rgba 0 0 0 0>}
      cornerRadii=<0 0 0 0>
     }
     tintColor {colorUseRule=UseColor_Specified, specifiedColor=<rgba 1 1 1 1>}
     uVRegion {bIsValid=false, max=<0 0>, min=<0 0>}
    }
    colorAndOpacity=<rgba 1 0 0 1>
    renderTransform {angle=0, scale=<1 1>, shear=<0 0>, translation=<-39 51>}
   }
   slot {class=OverlaySlot, properties {horizontalAlignment=HAlign_Center, verticalAlignment=VAlign_Center}}
  }
  Image_LiveBG {
   widgetClassPath=&Image
   parent=Overlay_3
   widget=Image_LiveBG
   properties {
    brush {
     bIsDynamicallyLoaded=false
     drawAs=Image
     imageType=NoImage
     mirroring=NoMirror
     resourceName=None
     resourceObject=&M_UI_SkewBox_LiveBG
     tiling=NoTile
     imageSize=<32 32>
     margin=<ltrb 0 0 0 0>
     outlineSettings {
      bUseBrushTransparency=false
      roundingType=HalfHeightRadius
      width=0
      color {colorUseRule=UseColor_Specified, specifiedColor=<rgba 0 0 0 0>}
      cornerRadii=<0 0 0 0>
     }
     tintColor {colorUseRule=UseColor_Specified, specifiedColor=<rgba 1 1 1 1>}
     uVRegion {bIsValid=false, max=<0 0>, min=<0 0>}
    }
    colorAndOpacity=<rgba 1 0 0 1>
    renderTransform {angle=0, scale=<1 1>, shear=<-14 0>, translation=<0 0>}
   }
   slot {class=OverlaySlot, properties {horizontalAlignment=HAlign_Fill, verticalAlignment=VAlign_Fill}}
  }
  HorizontalBox_0 {
   widgetClassPath=&HorizontalBox
   parent=Overlay_1
   widget=HorizontalBox_0
   slot {class=OverlaySlot, properties {padding=<ltrb 15 0 23 0>}}
  }
  Image_RecDot {
   widgetClassPath=&Image
   parent=HorizontalBox_0
   widget=Image_RecDot
   properties {brush {
     bIsDynamicallyLoaded=false
     drawAs=Image
     imageType=NoImage
     mirroring=NoMirror
     resourceName=None
     resourceObject=/Game/UI/Materials/M_UICircle
     tiling=NoTile
     imageSize=<16 16>
     margin=<ltrb 0 0 0 0>
     outlineSettings {
      bUseBrushTransparency=false
      roundingType=HalfHeightRadius
      width=0
      color {colorUseRule=UseColor_Specified, specifiedColor=<rgba 0 0 0 0>}
      cornerRadii=<0 0 0 0>
     }
     tintColor {colorUseRule=UseColor_Specified, specifiedColor=<rgba 1 1 1 1>}
     uVRegion {bIsValid=false, max=<0 0>, min=<0 0>}
    }}
   slot {class=HorizontalBoxSlot, properties {horizontalAlignment=HAlign_Center, verticalAlignment=VAlign_Center}}
  }
  Text_Callout {
   widgetClassPath=&CommonTextBlock
   bIsVariable=true
   parent=HorizontalBox_0
   widget=Text_Callout
   properties {
    text="LIVE!"
    font {
     bForceMonospaced=false
     bMaterialIsStencil=false
     fontMaterial=null
     fontObject=&Burbank
     letterSpacing=0
     monospacedWidth=1
     size=35
     skewAmount=0
     typefaceFontName=Burbank-Black
     outlineSettings {
      bApplyOutlineToDropShadows=false
      bMiteredCorners=false
      bSeparateFillAlpha=false
      outlineMaterial=null
      outlineSize=0
      outlineColor=<rgba 0 0 0 1>
     }
    }
    renderTransform {angle=0, scale=<1 1>, shear=<-10 0>, translation=<0 0>}
   }
   slot {class=HorizontalBoxSlot, properties {padding=<ltrb 10 0 0 0>}}
  }
 ]
}
